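# Unlit translucent, two-sided outline material for Niagara mesh particles in the Teriyaki Toon skydiving contrail effect.
# Screen-space edge detection compares neighboring SceneDepth and, optionally on the default shading path, WorldNormal samples, with distance-scaled thickness and a pixel-depth cull.
# Switchable world position offset combines normal inflation with either a fish-swing function or layered local-Z-masked sine waves.
# Dynamic parameters drive the waves, with PreviousFrameSwitch supplying previous-frame motion.
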Material /BRCosmetics/Athena/Effects/Particles/Skydiving/NS_Contrail_TeriyakiToon/Materials/M_FxSDOutLine
graph {
 properties {
  bEnableResponsiveAA=true
  bUsedWithNiagaraMeshParticles=true
  blendMode=BLEND_Translucent
  pixelDepthOffsetMode=PDOM_Legacy
  shadingModel=MSM_Unlit
  twoSided=true
  usedShadingModels=Unlit
 }
 nodes [
  n1 = DynamicParameter() {defaultValue=<rgba 0 1 1 2>, paramNames [Time Offset, Speed, Pulse, Frequency]}
  n2 = Multiply(A=@n1.Frequency)
  n3 = TransformPosition(None=WorldPosition().XYZ) {transformSourceType=TRANSFORMPOSSOURCE_World, transformType=TRANSFORMPOSSOURCE_Instance}
  n4 = Saturate(None=Divide(A=ComponentMask(None=@n3) {b=true}, B=ScalarParameter() {defaultValue=300, group=WPO, parameterName=WPO Object Size}) {constB=200})
  n5 = Power(Base=@n4, Exp=ScalarParameter() {defaultValue=16, group=WPO, parameterName=Power Hand Mask}) {constExponent=12}
  n6 = ScalarParameter() {defaultValue=2, group=WPO, parameterName=WPO Speed}
  n7 = Multiply(A=@n6) {constB=0.8}
  n8 = DeltaTime()
  n9 = LinearInterpolate(Alpha=Saturate(None=@n1.Pulse)) {constA=1, constB=2}
  n10 = PreviousFrameSwitch(Current Frame=@n1.Speed, Previous Frame=Subtract(A=@n1.Speed, B=Multiply(A=@n8, B=@n9)))
  n11 = Subtract(
   A=Multiply(A=LinearInterpolate(A=@n1.Frequency, B=@n2, Alpha=@n5) {constA=1, constB=4}, B=@n4)
   B=Add(A=@n1.Time Offset, B=Multiply(A=@n7, B=@n10))
  )
  Cosine(None=@n11)
  n12 = Multiply(A=@n1.Frequency) {constB=3}
  n13 = Multiply(A=@n6) {constB=0.9}
  n14 = Subtract(
   A=Multiply(A=LinearInterpolate(A=@n1.Frequency, B=@n12, Alpha=@n5) {constA=1, constB=4}, B=@n4)
   B=Add(A=@n1.Time Offset, B=Multiply(A=@n13, B=@n10))
  )
  Cosine(None=@n14)
  MaterialFunctionCall(Emissive=ScalarParameter() {parameterName=C}) {materialFunction=&MF_FX_EmissiveAdjustments}
  MaterialFunctionCall() {materialFunction=/Engine/Functions/Engine_MaterialFunctions02/Utility/FOV}
  n15 = ComponentMask(None=ScreenPosition().ViewportUV) {g=true, r=true}
  n16 = PixelDepth()
  n17 = ScalarParameter() {defaultValue=1.5, parameterName=Cull}
  n18 = Add(A=@n16.R, B=@n17)
  n19 = Multiply(A=@n17, B=@n16.R)
  n20 = Constant() {r=1}
  n21 = ScreenPosition()
  n22 = ScalarParameter() {defaultValue=1, parameterName=EdgeThicknessN}
  n23 = Subtract(A=CameraPositionWS(), B=StaticSwitchParameter(True=ObjectPositionWS(), False=ParticlePositionWS()) {parameterName=UseObjectPos})
  n24 = MaterialFunctionCall!(OffsetAmount=Divide(
    A=Multiply(A=@n22) {constB=-1}
    B=MaterialFunctionCall(Vector 3=@n23) {materialFunction=/Engine/Functions/Engine_MaterialFunctions02/Utility/VectorLength}.V3 Length
   )) {materialFunction=/Engine/Functions/Engine_MaterialFunctions01/ImageAdjustment/BlurSampleOffsets}
  n25 = SceneTexture(UVs=Add(A=@n21.ViewportUV, B=@n24."-.5,.5")) {sceneTextureId=PPI_WorldNormal}
  n26 = SceneTexture(UVs=@n21.ViewportUV) {sceneTextureId=PPI_WorldNormal}
  n27 = SceneTexture(UVs=Add(A=@n21.ViewportUV, B=@n24.".5,.5")) {sceneTextureId=PPI_WorldNormal}
  n28 = Add(A=@n21.ViewportUV, B=@n24.".5,-.5")
  n29 = Min(
   A=Min(A=DotProduct(A=@n25.Color, B=@n26.Color), B=DotProduct(A=@n27.Color, B=@n26.Color))
   B=DotProduct(A=SceneTexture(UVs=@n28) {sceneTextureId=PPI_WorldNormal}.Color, B=@n26.Color)
  )
  n30 = DotProduct(A=SceneTexture(UVs=Add(A=@n21.ViewportUV, B=@n24."-.5,-.5")) {sceneTextureId=PPI_WorldNormal}.Color, B=@n26.Color)
  n31 = Constant() {r=1}
  n32 = Subtract(A=Divide(A=ViewSize(), B=Constant() {r=320}), B=Constant() {r=2}) {constB=0}
  n33 = ViewSize()
  n34 = Divide(A=Divide(A=ComponentMask(None=@n33) {r=true}, B=ComponentMask(None=@n33) {g=true}), B=Constant() {r=1})
  n35 = ObjectPositionWS()
  n36 = ParticlePositionWS()
  n37 = Multiply(A=Distance(A=StaticSwitchParameter(True=@n35, False=@n36) {parameterName=UseObjectPos}, B=CameraPositionWS())) {constB=0.05}
  n38 = Multiply(
   A=Multiply(A=ScalarParameter() {defaultValue=1, parameterName=EdgeThickness}) {constB=15}
   B=Power(Base=Divide(A=@n34, B=@n37)) {constExponent=0.6}
  )
  n39 = Multiply(A=@n38, B=ScalarParameter() {defaultValue=0.775, group=Cartoon, parameterName=CRTNO_Outline-InnerSizeMobileMultiplier})
  n40 = Multiply(
   A=Clamp(None=Ceil(None=@n32)) {maxDefault=2, minDefault=1}
   B=Multiply(A=SceneTexelSize(), B=ShadingPathSwitch(Default=@n38, Forward=@n39, Mobile=@n39))
  ) {constA=2}
  n41 = Constant()
  n42 = AppendVector(A=ComponentMask(None=@n40) {desc=x, r=true}, B=@n41)
  n43 = Add(A=@n42, B=@n15)
  n44 = SceneDepth(UVs=@n15)
  n45 = AppendVector(A=@n41, B=ComponentMask(None=@n40) {desc=y, g=true}) {desc=y+}
  n46 = Add(A=@n15, B=@n45)
  n47 = Max(A=Subtract(A=SceneDepth(UVs=@n43).R, B=@n44.R), B=Subtract(A=SceneDepth(UVs=@n46).R, B=@n44.R))
  n48 = Constant() {r=-1}
  n49 = Multiply(A=@n42, B=@n48)
  n50 = Subtract(A=SceneDepth(UVs=Add(A=@n15, B=@n49)).R, B=@n44.R)
  n51 = SceneDepth(UVs=Add(A=@n15, B=Multiply(A=@n48, B=@n45)))
  n52 = Constant()
  n53 = If(
   A=Max(A=Max(A=@n47, B=@n50), B=Subtract(A=@n51.R, B=@n44.R))
   B=ScalarParameter() {defaultValue=15, parameterName=DepthT}
   "A > B"=Constant() {r=1}
   "A == B"=@n52
   "A < B"=@n52
  )
  n54 = Max(A=If(A=Min(A=@n29, B=@n30), B=ScalarParameter() {parameterName=NormalT}, "A > B"=Constant(), "A == B"=@n31, "A < B"=@n31), B=@n53)
  Multiply(
   A=If(
    A=SceneDepth(UVs=@n15).R
    B=StaticSwitchParameter(True=@n18, False=@n19) {parameterName=UseNewCull}
    "A > B"=Constant()
    "A == B"=@n20
    "A < B"=@n20
   )
   B=ShadingPathSwitch(Default=StaticSwitchParameter(True=@n54, False=@n53) {parameterName=UseNormalOutLIne}, Forward=@n53, Mobile=@n53)
  )
  n55 = ScalarParameter() {defaultValue=10, group=WPO, parameterName=WPO Amplitude}
  n56 = LinearInterpolate(A=@n55, B=Multiply(A=@n55) {constB=0.45}, Alpha=@n5) {constA=2, constB=1.5}
  n57 = ScalarParameter() {defaultValue=2, group=WPO, parameterName=WPO Frequency}
  n58 = Multiply(A=@n57) {constB=1.6}
  n59 = Subtract(A=Multiply(A=LinearInterpolate(A=@n57, B=@n58, Alpha=@n5) {constA=1, constB=4}, B=@n4), B=Add(A=@n1.Time Offset, B=Multiply(A=@n10, B=@n6)))
  n60 = AppendVector(
   A=AppendVector(A=Sine(None=@n59), B=Cosine(None=@n59))
   B=Multiply(A=Sine(None=@n59) {period=3}, B=ScalarParameter() {defaultValue=0.5, group=WPO, parameterName=WPO Z Adjustment}) {constB=0.6}
  )
  n61 = ScalarParameter() {defaultValue=2, group=WPO, parameterName=WPO Amplitude 2}
  n62 = Multiply(A=@n61) {constB=2}
  n63 = Constant()
  n64 = AppendVector(A=AppendVector(A=Sine(None=@n14), B=@n63), B=@n63)
  n65 = Multiply(A=@n61)
  n66 = Constant()
  n67 = AppendVector(A=Sine(None=@n11), B=@n66)
  n68 = Add(
   A=Add(A=Multiply(A=@n56, B=@n60) {constB=3}, B=Multiply(A=@n62, B=@n64) {constB=3})
   B=Multiply(A=LinearInterpolate(A=@n61, B=@n65, Alpha=@n5) {constA=2, constB=1.5}, B=AppendVector(A=@n67, B=@n66)) {constB=3}
  )
  n69 = StaticSwitchParameter(
   True=Transform(None=Multiply(A=@n4, B=@n68)) {transformSourceType=TRANSFORMSOURCE_Instance}
   False=MaterialFunctionCall() {materialFunction=&/MF_TeriyakiToonTrail_FishSwing}.Result
  ) {group=BalloonWPO, parameterName=UseBalloonWPO}
  n70 = Multiply(A=VertexNormalWS(), B=ScalarParameter() {parameterName=WPO})
  Reroute(None=StaticSwitchParameter(True=Add(A=@n69, B=@n70), False=@n70) {parameterName=UseSwingWPO})
  Time()
 ]
}
