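# Base emissive toon material for the TeriyakiToon skydiving contrail's Niagara mesh particles.
# A static switch chooses flat or particle color, or ink lines cut from a distance-field texture with blurred falloff and camera-distance-driven thickness, widened on mobile.
# The mobile path adds a contrasted-Fresnel fake outline in ink color.
# An optional Fresnel tint feeds in before the shared emissive adjustments function, and another switch provides fish-swing vertex motion.
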Material /BRCosmetics/Athena/Effects/Particles/Skydiving/NS_Contrail_TeriyakiToon/Materials/M_TeriyakiToonTrail_BaseEmissive
graph {
 properties {
  bUsedWithNiagaraMeshParticles=true
  pixelDepthOffsetMode=PDOM_Legacy
  previewMesh=/BRCosmetics/Gadgets/Assets/VinderTech_GliderChute/Glider_Ch_6_Mini_Season_2/FX/SM_Umbrella_Ch6S2_Spring
  usedShadingModels=Default Lit
 }
 nodes [
  Constant()
  n1 = VectorParameter() {parameterName=InkColor, defaultValue=<rgba 0 0 0 1>}
  n2 = VectorParameter() {parameterName=Color, defaultValue=<rgba 0 0 0 1>}
  n3 = TextureSampleParameter2D() {
   parameterName=DistanceField_InkLines
   samplerType=SAMPLERTYPE_Grayscale
   texture=/Game/Global/Textures/Default/Blanks/T_BlackLinearGrayscale_sRGB
  }
  n4 = Max(B=Multiply(A=ChannelMaskParameter(None=@n3.RGBA) {parameterName=DistanceField_Channel}.R) {desc=Mask Inkline Here}) {desc=Inject Inklines here}
  n5 = ScalarParameter() {defaultValue=9, parameterName=DF_BlurPow}
  n6 = OneMinus(None=Multiply(A=Power(Base=@n4, Exp=@n5) {constExponent=3}, B=ScalarParameter() {defaultValue=1, parameterName=DF_BlurMult}) {constB=0.5})
  n7 = Constant() {r=1}
  n8 = ScalarParameter() {defaultValue=100, group=NeonDepth, parameterName=DF_FarDistance}
  n9 = Subtract(A=Distance(A=WorldPosition().XYZ, B=CameraPositionWS()), B=@n8)
  n10 = Subtract(A=ScalarParameter() {defaultValue=500, group=NeonDepth, parameterName=DF_NearDistance}, B=@n8)
  n11 = LinearInterpolate(
   A=ScalarParameter() {defaultValue=0.08, group=NeonDepth, parameterName=DF_LineThickness_Near}
   B=ScalarParameter() {defaultValue=0.18, group=NeonDepth, parameterName=DF_LineThickness_Far}
   Alpha=Saturate(None=Divide(A=@n9, B=@n10))
  )
  n12 = ShadingPathSwitch(
   Default=@n11
   Deferred=@n11
   Forward=@n11
   Mobile=Multiply(A=@n11, B=ScalarParameter() {defaultValue=3.5, group=NeonDepth, parameterName=DF_ThicknessMobileScale})
  )
  n13 = ScalarParameter() {defaultValue=0.8, parameterName=DF_LineHardness}
  n14 = OneMinus(None=Saturate(None=SphereMask(A=@n4, B=@n7, Radius=@n12, Hardness=@n13) {attenuationRadius=0.1}))
  n15 = StaticSwitchParameter(
   True=LinearInterpolate(A=@n1.RGB, B=@n2.RGB, Alpha=Multiply(A=@n6, B=@n14))
   False=LinearInterpolate(A=@n2.RGB, B=ParticleColor().RGB, Alpha=ScalarParameter() {parameterName=UseParticleColor})
  ) {parameterName=UseInkDF}
  n16 = ScalarParameter() {defaultValue=1, parameterName=FakeOutline_Exp}
  n17 = Constant()
  n18 = LinearInterpolate(
   A=@n15
   B=@n1.RGB
   Alpha=MaterialFunctionCall(In=Fresnel(ExponentIn=@n16, BaseReflectFractionIn=@n17), Contrast=ScalarParameter() {defaultValue=50, parameterName=FakeOutline_Contra}) {materialFunction=&CheapContrast}.Result
  )
  n19 = Multiply(
   A=VectorParameter() {parameterName=FresnelColor, defaultValue=<rgba 0.0513 0.2462 0.1779 1>}.RGB
   B=Fresnel(
    ExponentIn=ScalarParameter() {defaultValue=6, parameterName=Fresnel_Exp}
    BaseReflectFractionIn=ScalarParameter() {defaultValue=0.02, parameterName=FresnelRef}
   ) {baseReflectFraction=0.02, exponent=6}
  )
  n20 = ScalarParameter() {parameterName=UseFresnel?}
  MaterialFunctionCall(Emissive=Add(A=ShadingPathSwitch(Default=@n15, Forward=@n15, Mobile=@n18), B=LinearInterpolate(B=@n19, Alpha=@n20) {constAlpha=1})) {materialFunction=&MF_FX_EmissiveAdjustments}
  ScalarParameter() {defaultValue=0.15, parameterName=lineThickness}
  StaticSwitchParameter(
   True=MaterialFunctionCall() {materialFunction=&/MF_TeriyakiToonTrail_FishSwing}.Result
   False=Constant()
  ) {parameterName=UseSwingWPO}
 ]
}
