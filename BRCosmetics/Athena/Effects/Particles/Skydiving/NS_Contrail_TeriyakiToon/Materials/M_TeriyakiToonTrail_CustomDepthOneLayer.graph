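# Translucent unlit Niagara mesh material for the TeriyakiToon skydiving contrail, drawing a cartoon fish with ink lines and outline.
# A distance-field texture forms ink lines that vary in thickness with camera distance, blended over the inner color.
# Four neighboring custom-depth samples, offset by distance, FOV and resolution, detect inner-edge outlines, masked by a depth cull.
# The result feeds an emissive-adjustment function, and a switch offers fish-swing vertex motion.
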Material /BRCosmetics/Athena/Effects/Particles/Skydiving/NS_Contrail_TeriyakiToon/Materials/M_TeriyakiToonTrail_CustomDepthOneLayer
graph {
 properties {
  allowTranslucentCustomDepthWrites=true
  bOutputTranslucentVelocity=true
  bUsedWithNiagaraMeshParticles=true
  blendMode=BLEND_Translucent
  pixelDepthOffsetMode=PDOM_Legacy
  shadingModel=MSM_Unlit
  usedShadingModels=Unlit
 }
 nodes [
  ChannelMaskParameter() {parameterName=DistanceField_Channel}
  Constant()
  n1 = VectorParameter() {parameterName=InnerColor, defaultValue=<rgba 0.7 0.7 0.7 1>}
  n2 = TextureSampleParameter2D() {
   parameterName=DistanceField_InkLines
   samplerType=SAMPLERTYPE_Masks
   texture=/BRCosmetics/Athena/Effects/Particles/Skydiving/NS_Contrail_TeriyakiToon/Texture/T_TeriyakiFishToon_Dive_DF
  }
  n3 = Max(B=Multiply(A=@n2.R) {desc=Mask Inkline Here}) {desc=Inject Inklines here}
  n4 = Multiply(
   A=Power(Base=@n3, Exp=ScalarParameter() {defaultValue=9, parameterName=DF_BlurPow}) {constExponent=3}
   B=ScalarParameter() {defaultValue=1, parameterName=DF_BlurMult}
  ) {constB=0.5}
  n5 = ScalarParameter() {defaultValue=0.05, group=NeonDepth, parameterName=DF_LineThickness_Near}
  n6 = ScalarParameter() {defaultValue=0.15, group=NeonDepth, parameterName=DF_LineThickness_Far}
  n7 = Distance(A=WorldPosition().XYZ, B=CameraPositionWS())
  n8 = ScalarParameter() {defaultValue=100, group=NeonDepth, parameterName=DF_FarDistance}
  n9 = ScalarParameter() {defaultValue=500, group=NeonDepth, parameterName=DF_NearDistance}
  n10 = Saturate(None=Divide(A=Subtract(A=@n7, B=@n8), B=Subtract(A=@n9, B=@n8)))
  n11 = Saturate(None=SphereMask(
    A=@n3
    B=Constant() {r=1}
    Radius=LinearInterpolate(A=@n5, B=@n6, Alpha=@n10)
    Hardness=ScalarParameter() {defaultValue=0.8, parameterName=DF_LineHardness}
   ) {attenuationRadius=0.1})
  n12 = LinearInterpolate(
   A=VectorParameter() {parameterName=InkColor, defaultValue=<rgba 0 0 0 1>}.RGB
   B=@n1.RGB
   Alpha=Multiply(A=OneMinus(None=@n4), B=OneMinus(None=@n11))
  )
  n13 = ComponentMask(None=SceneTexture() {sceneTextureId=PPI_CustomDepth}.Color) {r=true}
  n14 = Constant() {r=1}
  n15 = If(
   A=@n13
   B=Multiply(A=ScalarParameter() {defaultValue=1.5, parameterName=CullDepth}, B=PixelDepth().R)
   "A > B"=Constant()
   "A == B"=@n14
   "A < B"=@n14
  )
  n16 = ScreenPosition()
  n17 = Subtract(A=Divide(A=ViewSize(), B=Constant() {r=320}), B=Constant() {r=2}) {constB=0}
  n18 = ViewSize()
  n19 = Divide(
   A=Divide(A=ComponentMask(None=@n18) {r=true}, B=ComponentMask(None=@n18) {g=true})
   B=MaterialFunctionCall() {materialFunction=/Engine/Functions/Engine_MaterialFunctions02/Utility/FOV}."FOV "
  )
  n20 = ObjectPositionWS()
  n21 = ParticlePositionWS()
  n22 = Multiply(A=Distance(A=StaticSwitchParameter(True=@n20, False=@n21) {defaultValue=true, parameterName=UseObjectPos}, B=CameraPositionWS())) {constB=0.05}
  n23 = Multiply(
   A=ScalarParameter() {defaultValue=8, group=Cartoon, parameterName=CRTNO_Outline-InnerSize1}
   B=Power(Base=Divide(A=@n19, B=@n22)) {constExponent=0.6}
  )
  n24 = Multiply(A=@n23, B=ScalarParameter() {defaultValue=1.5, group=Cartoon, parameterName=CRTNO_Outline-InnerSizeMobileMultiplier})
  n25 = Multiply(
   A=Clamp(None=Ceil(None=@n17)) {maxDefault=2, minDefault=1}
   B=Multiply(A=SceneTexelSize(), B=ShadingPathSwitch(Default=@n23, Forward=@n24, Mobile=@n24))
  ) {constA=2}
  n26 = Constant()
  n27 = AppendVector(A=ComponentMask(None=@n25) {desc=x, r=true}, B=@n26)
  n28 = SceneTexture(UVs=Add(A=@n16.ViewportUV, B=@n27) {desc=x+}) {sceneTextureId=PPI_CustomDepth}
  n29 = AppendVector(A=@n26, B=ComponentMask(None=@n25) {desc=y, g=true}) {desc=y+}
  n30 = SceneTexture(UVs=Add(A=@n16.ViewportUV, B=@n29) {desc=y+}) {sceneTextureId=PPI_CustomDepth}
  n31 = Max(A=Subtract(A=ComponentMask(None=@n28.Color) {r=true}, B=@n13), B=Subtract(A=ComponentMask(None=@n30.Color) {r=true}, B=@n13))
  n32 = Constant() {r=-1}
  n33 = Add(A=@n16.ViewportUV, B=Multiply(A=@n27, B=@n32)) {desc=x-}
  n34 = Subtract(A=ComponentMask(None=SceneTexture(UVs=@n33) {sceneTextureId=PPI_CustomDepth}.Color) {r=true}, B=@n13)
  n35 = Multiply(A=@n32, B=@n29)
  n36 = ComponentMask(None=SceneTexture(UVs=Add(A=@n16.ViewportUV, B=@n35) {desc=y-}) {sceneTextureId=PPI_CustomDepth}.Color) {r=true}
  n37 = Constant()
  n38 = If(
   A=Max(A=Max(A=@n31, B=@n34), B=Subtract(A=@n36, B=@n13))
   B=ScalarParameter() {defaultValue=20, parameterName=DepthCompareT}
   "A > B"=Constant() {r=1}
   "A == B"=@n37
   "A < B"=@n37
  )
  MaterialFunctionCall(Emissive=LinearInterpolate(
    A=StaticSwitchParameter(True=@n12, False=@n1.RGB) {defaultValue=true, parameterName=UseInkDF}
    B=VectorParameter() {parameterName=OutColor, defaultValue=<rgba 0 0 0 1>}.RGB
    Alpha=Multiply(A=@n15, B=@n38)
   )) {materialFunction=&MF_FX_EmissiveAdjustments}
  ScalarParameter() {defaultValue=0.15, parameterName=lineThickness}
  StaticSwitchParameter(
   True=MaterialFunctionCall() {materialFunction=&/MF_TeriyakiToonTrail_FishSwing}.Result
   False=Constant()
  ) {parameterName=UseSwingWPO}
 ]
}
