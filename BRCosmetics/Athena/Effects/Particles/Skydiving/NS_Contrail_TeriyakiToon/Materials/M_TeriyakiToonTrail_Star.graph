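# Translucent, unlit, two-sided Niagara sprite material for the star particles of the TeriyakiToon skydiving contrail.
# It samples a star atlas texture, raises its color to a power, scales it down slightly and passes it through MF_FX_EmissiveAdjustments.
# The atlas alpha runs through CheapContrast to sharpen the star shapes.
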Material /BRCosmetics/Athena/Effects/Particles/Skydiving/NS_Contrail_TeriyakiToon/Materials/M_TeriyakiToonTrail_Star
graph {
 properties {
  bUsedWithNiagaraSprites=true
  blendMode=BLEND_Translucent
  pixelDepthOffsetMode=PDOM_Legacy
  shadingModel=MSM_Unlit
  twoSided=true
  usedShadingModels=Unlit
 }
 nodes [
  n1 = TextureSample() {texture=/BRCosmetics/Athena/Effects/Particles/Skydiving/NS_Contrail_TeriyakiToon/Texture/T_StarsAtlas}
  MaterialFunctionCall(In=@n1.A, Contrast=Constant() {r=1}) {materialFunction=&CheapContrast}
  MaterialFunctionCall(Emissive=Multiply(A=Power(Base=@n1.RGB)) {constB=0.75}) {materialFunction=&MF_FX_EmissiveAdjustments}
 ]
}
