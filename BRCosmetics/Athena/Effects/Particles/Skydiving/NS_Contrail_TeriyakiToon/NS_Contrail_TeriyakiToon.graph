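# Niagara skydiving contrail cosmetic (TeriyakiToon), with dive-driven system values blending its motion and emission.
# Two mesh and two sprite emitters spawn rising noisy trails on paired SmokeStar sockets, adding live socket positions to simulated offsets.
# A burst emitter pins left/right hand and fish mesh parts to character sockets, passing rotation axes and a timer to materials.
# Outline/fresnel values vary by platform; front-end preview tilts the effect.
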NiagaraSystem /BRCosmetics/Athena/Effects/Particles/Skydiving/NS_Contrail_TeriyakiToon/NS_Contrail_TeriyakiToon
dependencies {
 usedDataInterfaces [
  &NiagaraDataInterfacePlatformSet
  &NiagaraDataInterfaceCurve
  &NiagaraDataInterfaceCurlNoise
  &NiagaraDataInterfaceArrayFloat
 ]
 usedDynamicInputs [&ModuloFloat]
 usedModules [
  &self:SystemSpawnScript.SystemScriptSource.SystemScriptGraph.NiagaraNodeAssignment_0.SetVariables_#1
  &SystemState
  /Game/Effects/Niagara/Modules/Skydiving/FN_DivingAttributes/v4/FN_DivingAttributes_v4
  /Game/Effects/Niagara/Modules/Skydiving/FN_DivingAttributes/v6/FN_DivingAttributes
  &self:SystemSpawnScript.SystemScriptSource.SystemScriptGraph.NiagaraNodeAssignment_3.SetVariables_#2
  &self:SystemSpawnScript.SystemScriptSource.SystemScriptGraph.NiagaraNodeAssignment_1.SetVariables_#3
 ]
}
emitters {
 Empty {
  emitterUpdateScript [
   &EmitterState()
   &SpawnRate(SpawnRate=&Lerp_Float(A=3, B=14, Alpha=$System.RDive), Use Spawn Probability=false)
   /SetVariables(Emitter.EdgeThickness=0.6) {isSetParameters=true, module=SetVariables_#4}
  ]
  particleSpawnScript [
   &FN_SetSpawnGroup(Number of Spawn Groups=2)
   &InitializeParticle(
    Color Mode=::NewEnumerator0
    Lifetime=&Lerp_Float(
     A=&RandomRangeFloat(Minimum=1.7, Maximum=2.3)
     B=&RandomRangeFloat(Minimum=0.85, Maximum=1.05)
     Alpha=$System.RDive
    )
    Mesh Scale Mode=::NewEnumerator1
    Mesh Uniform Scale Max=0.7
    Mesh Uniform Scale Min=0.3
    Sprite Rotation Mode=::NewEnumerator2
    Sprite Size Mode=::NewEnumerator1
    Uniform Sprite Size Max=11
    Uniform Sprite Size Min=4.5
    UsePositionOffset=false
    Write Material Random=true
   )
   &AddVelocity(Velocity=&Lerp_Vector(A=<0 0 80>, B=<0 0 223>, Alpha=$System.RDive))
   &AddVelocity(Velocity=&MakeCustomVectorFromBool(
     Bool=$User.bIsFrontEndPreview
     True Vector=&Lerp_Vector(A=<0 0 0>, B=<0 -100 0>, Alpha=$System.Dive)
     False Vector=<0 0 0>
    )) {module=AddVelocity001}
   &InitialMeshOrientation(Use Rotation Vector=false)
   &SphereLocation(
    MaskSpawnGroup=false
    Override Seed=false
    Sphere Radius=30
    Surface Only=false
    UseNonUniformScale=false
    UseOffset=false
    UseSphereOrientation=false
   ) {enabled=false}
   &TorusLocation(Align Torus With Axis=false, Handle Radius=7.5, Large Radius=20, MaskSpawnGroup=false, Module.Override Randomness=false)
  ]
  particleUpdateScript [
   &ParticleState()
   &CurlNoiseForce(Noise Strength=30, Pan Noise=false)
   &SolveForcesAndVelocity(Clamp Velocity=false, Limit Acceleration=false)
   &SkeletalMeshLocation(
    Bone / Socket Index=$Particles.SpawnGroup
    Bone Sampling Mode=::NewEnumerator4
    MaskSpawnGroup=false
    Orientation Sampling=::NewEnumerator1
    Position Sampling=::NewEnumerator2
    Sampled Position Offset=&MakeCustomVectorFromBool(
     Bool=&SetBoolByIntComparison(A=$Particles.SpawnGroup, B=0, Comparison Type=::NewEnumerator4)
     True Vector=&Lerp_Vector(A=<-7 0 0>, B=<-15 1 1>, Alpha=$System.RDive)
     False Vector=&Lerp_Vector(A=<7 0 0>, B=<15 1 1>, Alpha=$System.RDive)
    )
    Sampled Position Offset Coordinate Space:ENiagaraSkelMeshTransforms=::NewEnumerator2
    Skeletal Mesh=$User.SmokeStarSkeletalMesh
   )
   /SetVariables(Particles.NPosition:Vector3f=&Add_Vector(
     A=Linked() {linkedVariable {name=Particles.Position, type=NiagaraPosition}}
     B=Linked() {linkedVariable {name=Particles.SkeletalMeshLocation.SampledPosition, type=NiagaraPosition}}
    )) {isSetParameters=true, module=SetVariables_#5}
   &ScaleMeshSize(Scale Factor=&VectorFromFloat(Value=&FloatFromCurve(FloatCurve=DataInterface() {propertyValues {Curve {keys [
          {interpMode=RCIM_Cubic, leaveTangent=5.3541, tangentMode=RCTM_Break, time=0, value=0}
          {interpMode=RCIM_Cubic, tangentMode=RCTM_Break, time=0.3171, value=1}
          {arriveTangent=-0.5305, interpMode=RCIM_Cubic, leaveTangent=-0.5305, tangentMode=RCTM_User, time=0.7, value=0.8}
          {arriveTangent=-4.7219, interpMode=RCIM_Cubic, leaveTangent=-4.7219, tangentMode=RCTM_User, time=1, value=0}
         ]}}})))
   &ScaleMeshSize(Scale Factor=<1 1.5 1>) {module=ScaleMeshSize001}
   /SetVariables(Particles.NScale:Vector3f=&Multiply_VectorByFloat(Vector=$Particles.Scale, Float=1.1)) {isSetParameters=true, module=SetVariables_#6}
   /SetVariables(Particles.MeshOrientation=&MakeQuatFromAxisAngle(Axis=<0 0 1>, AngleInDegrees=$System.Age)) {enabled=false, isSetParameters=true, module=SetVariables_#7}
   &UpdateMeshOrientation(Rotation Rate=0.5, Rotation Vector=&RandomVector(Vector Scale=1))
   &SubUVAnimation(End Frame=3, SubUV Lookup Index:NiagaraFloat=&RandomRangeFloat(Minimum=0, Maximum=1))
   &SpriteRotationRate(Rotation Rate=&Multiply_Float(
     A=&RandomRangeFloat(Minimum=150, Maximum=300)
     B=&SelectFloatFromArray(
      Array Sampling Mode=::NewEnumerator0
      Direct Array Index=?
      Float Selection Array=DataInterface() {propertyValues {FloatData [-1,1]}}
      Random Seed=0
     )
    ))
   &ScaleSpriteSize(Uniform Curve Sprite Scale=DataInterface() {propertyValues {Curve {keys [
        {interpMode=RCIM_Cubic, leaveTangent=8.7778, tangentMode=RCTM_Break, time=0, value=0}
        {interpMode=RCIM_Cubic, tangentMode=RCTM_Break, time=0.252, value=1}
        {arriveTangent=-0.5305, interpMode=RCIM_Cubic, leaveTangent=-0.5305, tangentMode=RCTM_User, time=0.7, value=0.8}
        {arriveTangent=-4.7219, interpMode=RCIM_Cubic, leaveTangent=-4.7219, tangentMode=RCTM_User, time=1, value=0}
       ]}}})
   /SetVariables() {isSetParameters=true, module=SetVariables_#8}
  ]
  rendererClasses [&NiagaraSpriteRendererProperties]
  renderers [{rendererClass=&NiagaraSpriteRendererProperties}]
 }
 Empty001 {
  emitterUpdateScript [
   &EmitterState(Life Cycle Mode=::NewEnumerator1, Loop Behavior=::NewEnumerator1, Loop Duration Mode=::NewEnumerator1, UseLoopDelay=false)
   &SpawnBurst_Instantaneous(Use Spawn Probability=false, UseLoopCountLimit=false)
   &SpawnBurst_Instantaneous(Spawn Group=1, Use Spawn Probability=false, UseLoopCountLimit=false) {module=SpawnBurst_Instantaneous001}
   &SpawnBurst_Instantaneous(Spawn Group=2, Use Spawn Probability=false, UseLoopCountLimit=false) {module=SpawnBurst_Instantaneous002}
   &SpawnBurst_Instantaneous(Spawn Group=3, Use Spawn Probability=false, UseLoopCountLimit=false) {module=SpawnBurst_Instantaneous003}
   &SpawnBurst_Instantaneous(Spawn Time=6.1524, Use Spawn Probability=false, UseLoopCountLimit=false) {enabled=false, module=SpawnBurst_Instantaneous004}
   /SetVariables(Emitter.FresnelThickEXP:NiagaraFloat=&self:PlatFromSelect(
     Default:NiagaraFloat=50
     IsTarget:NiagaraFloat=1.6
     TargetPlatform:NiagaraDataInterfacePlatformSet=DataInterface() {propertyValues {Platforms {
        qualityLevelMask=-32
        deviceProfileStates [
         {profileName=IOS, qualityLevelMask=4294967295, setQualityLevelMask=4294967295}
         {profileName=Android, qualityLevelMask=4294967295, setQualityLevelMask=4294967295}
         {profileName="Android_Mali_G72 ", qualityLevelMask=4294967295, setQualityLevelMask=4294967295}
         {profileName=Android_Galaxy_Z_Fold2_Adreno, qualityLevelMask=4294967295, setQualityLevelMask=4294967295}
         {profileName=Android_Galaxy_Z_Fold2_Adreno_LRU, qualityLevelMask=4294967295, setQualityLevelMask=4294967295}
         {profileName=Android_Galaxy_Z_Fold2_Adreno_Vulkan, qualityLevelMask=4294967295, setQualityLevelMask=4294967295}
         {profileName=Android_High, qualityLevelMask=4294967295, setQualityLevelMask=4294967295}
         {profileName=IOS_Epic, qualityLevelMask=4294967295, setQualityLevelMask=4294967295}
        ]
       }}}
    )) {isSetParameters=true}
  ]
  particleSpawnScript [
   &InitializeParticle(Color Mode=::NewEnumerator0, UsePositionOffset=false, Write Material Random=true)
   /SetVariables(Particles.SpawnGroup=&SpawnGroup()) {isSetParameters=true}
   /SetVariables(
    Particles.VisibilityTag=&MakeCustomIntFromBool(
     Bool=&SetBoolByIntComparison(A=$Particles.SpawnGroup, B=2, Comparison Type=::NewEnumerator6)
     True INT=0
     False INT=&MakeCustomIntFromBool(
      Bool=&SetBoolByIntComparison(A=$Particles.SpawnGroup, B=3, Comparison Type=::NewEnumerator4)
      True INT=2
      False INT=1
     )
    )
    Particles.isLeft:NiagaraBool=&Boolean_OrOperation(
     Bool1=&SetBoolByIntComparison(A=$Particles.SpawnGroup, B=0, Comparison Type=::NewEnumerator4)
     Bool2=&SetBoolByIntComparison(A=$Particles.SpawnGroup, B=2, Comparison Type=::NewEnumerator4)
    )
   ) {isSetParameters=true}
   /SetVariables(Particles.Scale=&MakeCustomVectorFromBool(
     Bool=&SetBoolByIntComparison(A=$Particles.SpawnGroup, B=2, Comparison Type=::NewEnumerator6)
     True Vector=<0.8 0.8 0.8>
     False Vector=&MakeCustomVectorFromBool(
      Bool=$Particles.isLeft
      True Vector=&VectorFromFloat(Value=1.1)
      False Vector=&VectorFromFloat(Value=1.1)
     )
    )) {isSetParameters=true}
  ]
  particleUpdateScript [
   &ParticleState(Kill Particles When Lifetime Has Elapsed=false, Let Infinitely Lived Particles Die When Emitter Deactivates=true)
   &SkeletalMeshLocation(
    Bone / Socket Index=&MakeCustomIntFromBool(Bool=$Particles.isLeft, True INT=0, False INT=1)
    Bone Sampling Mode=::NewEnumerator4
    MaskSpawnGroup=false
    Orientation Sampling=::NewEnumerator1
    Position Sampling=::NewEnumerator2
    Sampled Position Offset Coordinate Space:ENiagaraSkelMeshTransforms=::NewEnumerator4
    Skeletal Mesh=$User.SkeletalMesh
   )
   &SkeletalMeshLocation(
    Bone / Socket Index=&MakeCustomIntFromBool(Bool=$Particles.isLeft, True INT=2, False INT=3)
    Bone Sampling Mode=::NewEnumerator4
    MaskSpawnGroup=false
    Orientation Sampling=::NewEnumerator1
    Position Sampling=::NewEnumerator2
    Sampled Position Offset Coordinate Space:ENiagaraSkelMeshTransforms=::NewEnumerator4
    Skeletal Mesh=$User.SkeletalMesh
   ) {module=SkeletalMeshLocation001}
   /SetVariables(Particles.HandPart_Orientation:Quat4f=&MultiplyQuaternion(
     Quaternion A=&MultiplyQuaternion(
      Quaternion A=&MultiplyQuaternion(
       Quaternion A=$Particles.SkeletalMeshLocation001.SampledRotation
       Quaternion B=&MakeQuatFromAxisAngle(
        Axis=<0 0 1>
        AngleInDegrees=&MakeCustomFloatFromBool(Bool=$Particles.isLeft, True Float=90, False Float=-90)
       )
      )
      Quaternion B=&MakeQuatFromAxisAngle(
       Axis=<0 1 0>
       AngleInDegrees=&Lerp_Float(
        A=&MakeCustomFloatFromBool(Bool=$Particles.isLeft, True Float=-20, False Float=-160)
        B=&MakeCustomFloatFromBool(Bool=$Particles.isLeft, True Float=-100, False Float=-80)
        Alpha=$System.Dive
       )
      )
     )
     Quaternion B=&MakeQuatFromAxisAngle(Axis=<1 0 0>, AngleInDegrees=&Lerp_Float(A=0, B=45, Alpha=$System.Dive))
    )) {isSetParameters=true}
   /SetVariables(Particles.HandPart_Position:Vector3f=&Add_Vector(
     A=&Lerp_Vector(
      A=Linked() {linkedVariable {name=Particles.SkeletalMeshLocation.SampledPosition, type=NiagaraPosition}}
      B=Linked() {linkedVariable {name=Particles.SkeletalMeshLocation001.SampledPosition, type=NiagaraPosition}}
      Alpha=$System.Lerp
     )
     B=&Multiply_VectorByFloat(
      Vector=&RotateVectorByQuaternion(Vector To Rotate=<0 0 0>, Quaternion=$Particles.HandPart_Orientation)
      Float=2
     )
    )) {isSetParameters=true}
   /SetVariables(Particles.FishPart_Orientation:Quat4f=&MultiplyQuaternion(
     Quaternion A=&MultiplyQuaternion(
      Quaternion A=&MakeQuatFromAxisAngle(
       Axis=<0 0 1>
       AngleInDegrees=&MakeCustomFloatFromBool(Bool=$Particles.isLeft, True Float=70, False Float=120)
      )
      Quaternion B=&MakeQuatFromAxisAngle(Axis=<0 1 0>, AngleInDegrees=&Lerp_Float(A=0, B=15, Alpha=$System.Dive))
     )
     Quaternion B=&MakeQuatFromAxisAngle(
      Axis=<1 0 -1>
      AngleInDegrees=&Lerp_Float(
       A=0
       B=&MakeCustomFloatFromBool(Bool=$Particles.isLeft, True Float=20, False Float=-20)
       Alpha=$System.Dive
      )
      NormalizeAxis=true
     )
    )) {isSetParameters=true}
   /SetVariables(Particles.FishPart_Orientation:Quat4f=&MultiplyQuaternion(
     Quaternion A=$Particles.FishPart_Orientation
     Quaternion B=&MakeQuatFromAxisAngle(
      Axis=&MakeCustomVectorFromBool(Bool=$Particles.isLeft, True Vector=<-1.5 1 0>, False Vector=<1.5 1 0>)
      AngleInDegrees=&Lerp_Float(
       A=0
       B=&MakeCustomFloatFromBool(Bool=$User.bIsFrontEndPreview, True Float=-50, False Float=0)
       Alpha=$System.Dive
      )
      NormalizeAxis=true
     )
    )) {isSetParameters=true}
   /SetVariables(Particles.FishPart_Position:Vector3f=&Add_Vector(
     A=&Lerp_Vector(
      A=Linked() {linkedVariable {name=Particles.SkeletalMeshLocation.SampledPosition, type=NiagaraPosition}}
      B=Linked() {linkedVariable {name=Particles.SkeletalMeshLocation001.SampledPosition, type=NiagaraPosition}}
      Alpha=$System.Lerp
     )
     B=&Add_Vector(
      A=&Multiply_VectorByFloat(
       Vector=&RotateVectorByQuaternion(Vector To Rotate=<0 0 1>, Quaternion=$Particles.HandPart_Orientation)
       Float=10
      )
      B=<0 0 0>
     )
    )) {isSetParameters=true}
   /SetVariables(Particles.MaterialRotateAxis:Vector3f=&TransformVector(
     Vector=&RotateVectorByQuaternion(Vector To Rotate=<0 0 1>, Quaternion=$Particles.FishPart_Orientation)
     Destination Space=::World
    )) {isSetParameters=true}
   /SetVariables(Particles.MaterialRotate1Axis:Vector3f=&TransformVector(
     Vector=&RotateVectorByQuaternion(
      Vector To Rotate=&MakeCustomVectorFromBool(Bool=$Particles.isLeft, True Vector=<-0.8 0 0>, False Vector=<0.8 0 0>)
      Quaternion=$Particles.FishPart_Orientation
     )
     Destination Space=::World
    )) {isSetParameters=true}
   &FN_Timer(Increment Multiplier:NiagaraFloat=&Lerp_Float(A=1, B=1.5, Alpha=$System.Dive))
   &DynamicMaterialParameters(
    Index 0 Param 1=&MakeFloatFromVector(Vector=$Particles.MaterialRotateAxis, Channel=::NewEnumerator0)
    Index 0 Param 2=&MakeFloatFromVector(Vector=$Particles.MaterialRotateAxis, Channel=::NewEnumerator1)
    Index 0 Param 3=&MakeFloatFromVector(Vector=$Particles.MaterialRotateAxis, Channel=::NewEnumerator2)
    Index 0 Param 4=&RandomRangeFloat(Minimum=0, Maximum=2)
    Index 1 Param 1=&MakeFloatFromVector(Vector=$Particles.MaterialRotate1Axis, Channel=::NewEnumerator0)
    Index 1 Param 2=&MakeFloatFromVector(Vector=$Particles.MaterialRotate1Axis, Channel=::NewEnumerator1)
    Index 1 Param 3=&MakeFloatFromVector(Vector=$Particles.MaterialRotate1Axis, Channel=::NewEnumerator2)
    Index 1 Param 4=&Lerp_Float(A=0.2, B=1, Alpha=$System.Dive)
    Index 2 Param 1=$Particles.FN_Timer.Value
    Module.Param0WriteEnabled=true
    Module.Param0WriteEnabled1=true
    Module.Param0WriteEnabled2=true
    Module.Param1WriteEnabled=true
    Module.Param1WriteEnabled1=true
    Module.Param1WriteEnabled2=true
    Module.Param2WriteEnabled=true
    Module.Param2WriteEnabled1=true
    Module.Param2WriteEnabled2=true
    Module.Param3WriteEnabled=true
    Module.Param3WriteEnabled1=true
    Module.Param3WriteEnabled2=true
    Write Parameter Index 1=true
    Write Parameter Index 2=true
   )
  ]
  rendererClasses [&NiagaraMeshRendererProperties]
  renderers [
   {rendererClass=&NiagaraMeshRendererProperties, rendererIndex=0}
   {rendererClass=&NiagaraMeshRendererProperties, rendererIndex=1}
   {rendererClass=&NiagaraMeshRendererProperties, rendererIndex=2}
   {rendererClass=&NiagaraMeshRendererProperties, rendererIndex=3}
   {rendererClass=&NiagaraMeshRendererProperties, rendererIndex=4}
   {rendererClass=&NiagaraMeshRendererProperties, rendererIndex=5}
  ]
 }
 Empty002 {
  emitterUpdateScript [
   &EmitterState()
   &SpawnRate(SpawnRate=&Lerp_Float(A=1.5, B=2, Alpha=$System.RDive), Use Spawn Probability=false)
   /SetVariables(Emitter.EdgeThickness=0.6) {isSetParameters=true, module=SetVariables_#4}
  ]
  particleSpawnScript [
   &FN_SetSpawnGroup(Number of Spawn Groups=2)
   &InitializeParticle(
    Color Mode=::NewEnumerator0
    Lifetime=&Lerp_Float(
     A=&RandomRangeFloat(Minimum=1.5, Maximum=2.1)
     B=&RandomRangeFloat(Minimum=0.75, Maximum=0.9)
     Alpha=$System.RDive
    )
    Mesh Scale Mode=::NewEnumerator1
    Mesh Uniform Scale Max=0.45
    Mesh Uniform Scale Min=0.25
    Sprite Rotation Mode=::NewEnumerator2
    Sprite Size Mode=::NewEnumerator1
    Uniform Sprite Size Max=5
    Uniform Sprite Size Min=3
    UsePositionOffset=false
    Write Material Random=true
   )
   &AddVelocity(Velocity=&Lerp_Vector(A=<0 0 50>, B=<0 0 180>, Alpha=$System.RDive))
   &AddVelocity(Velocity=&MakeCustomVectorFromBool(
     Bool=$User.bIsFrontEndPreview
     True Vector=&Lerp_Vector(A=<0 0 0>, B=<0 -50 0>, Alpha=$System.Dive)
     False Vector=<0 0 0>
    )) {module=AddVelocity001}
   &InitialMeshOrientation(Use Rotation Vector=false)
   &SphereLocation(
    MaskSpawnGroup=false
    Override Seed=false
    Sphere Radius=30
    Surface Only=false
    UseNonUniformScale=false
    UseOffset=false
    UseSphereOrientation=false
   ) {enabled=false}
   &TorusLocation(Align Torus With Axis=false, Handle Radius=2, Large Radius=7, MaskSpawnGroup=false, Module.Override Randomness=false)
  ]
  particleUpdateScript [
   &ParticleState()
   &CurlNoiseForce(Noise Strength=2, Pan Noise=false)
   &SolveForcesAndVelocity(Clamp Velocity=false, Limit Acceleration=false)
   &SkeletalMeshLocation(
    Bone / Socket Index=&Add_Integer(A=$Particles.SpawnGroup, B=2)
    Bone Sampling Mode=::NewEnumerator4
    MaskSpawnGroup=false
    Orientation Sampling=::NewEnumerator1
    Position Sampling=::NewEnumerator2
    Sampled Position Offset=&MakeCustomVectorFromBool(
     Bool=&SetBoolByIntComparison(A=$Particles.SpawnGroup, B=0, Comparison Type=::NewEnumerator4)
     True Vector=<0 0 0>
     False Vector=<0 0 0>
    )
    Sampled Position Offset Coordinate Space:ENiagaraSkelMeshTransforms=::NewEnumerator2
    Skeletal Mesh=$User.SmokeStarSkeletalMesh
   )
   /SetVariables(Particles.NPosition:Vector3f=&Add_Vector(
     A=Linked() {linkedVariable {name=Particles.Position, type=NiagaraPosition}}
     B=Linked() {linkedVariable {name=Particles.SkeletalMeshLocation.SampledPosition, type=NiagaraPosition}}
    )) {isSetParameters=true, module=SetVariables_#5}
   &ScaleMeshSize(Scale Factor=&VectorFromFloat(Value=&FloatFromCurve(FloatCurve=DataInterface() {propertyValues {Curve {keys [
          {interpMode=RCIM_Cubic, leaveTangent=5.3541, tangentMode=RCTM_Break, time=0, value=0}
          {interpMode=RCIM_Cubic, tangentMode=RCTM_Break, time=0.3171, value=1}
          {arriveTangent=-0.5305, interpMode=RCIM_Cubic, leaveTangent=-0.5305, tangentMode=RCTM_User, time=0.7, value=0.8}
          {arriveTangent=-4.7219, interpMode=RCIM_Cubic, leaveTangent=-4.7219, tangentMode=RCTM_User, time=1, value=0}
         ]}}})))
   &ScaleMeshSize(Scale Factor=<1 1.5 1>) {module=ScaleMeshSize001}
   /SetVariables(Particles.NScale:Vector3f=&Multiply_VectorByFloat(Vector=$Particles.Scale, Float=1.1)) {isSetParameters=true, module=SetVariables_#6}
   /SetVariables(Particles.MeshOrientation=&MakeQuatFromAxisAngle(Axis=<0 0 1>, AngleInDegrees=$System.Age)) {enabled=false, isSetParameters=true, module=SetVariables_#7}
   &UpdateMeshOrientation(Rotation Rate=0.5, Rotation Vector=&RandomVector(Vector Scale=1))
   &SubUVAnimation(End Frame=3, SubUV Lookup Index:NiagaraFloat=&RandomRangeFloat(Minimum=0, Maximum=1))
   &SpriteRotationRate(Rotation Rate=&Multiply_Float(
     A=&RandomRangeFloat(Minimum=150, Maximum=300)
     B=&SelectFloatFromArray(
      Array Sampling Mode=::NewEnumerator0
      Direct Array Index=?
      Float Selection Array=DataInterface() {propertyValues {FloatData [-1,1]}}
      Random Seed=0
     )
    ))
   &ScaleSpriteSize(Uniform Curve Sprite Scale=DataInterface() {propertyValues {Curve {keys [
        {interpMode=RCIM_Cubic, leaveTangent=8.7778, tangentMode=RCTM_Break, time=0, value=0}
        {interpMode=RCIM_Cubic, tangentMode=RCTM_Break, time=0.252, value=1}
        {arriveTangent=-0.5305, interpMode=RCIM_Cubic, leaveTangent=-0.5305, tangentMode=RCTM_User, time=0.7, value=0.8}
        {arriveTangent=-4.7219, interpMode=RCIM_Cubic, leaveTangent=-4.7219, tangentMode=RCTM_User, time=1, value=0}
       ]}}})
   /SetVariables() {isSetParameters=true, module=SetVariables_#8}
  ]
  rendererClasses [&NiagaraSpriteRendererProperties]
  renderers [{rendererClass=&NiagaraSpriteRendererProperties}]
 }
 Fountain {
  emitterUpdateScript [
   &EmitterState(Life Cycle Mode=::NewEnumerator1, Loop Duration=2, UseLoopDelay=false)
   &SpawnRate(SpawnRate=&Lerp_Float(A=28, B=53, Alpha=$System.RDive), Use Spawn Probability=false)
   /SetVariables(
    Emitter.EdgeThickness=0.55
    Emitter.FresnelThickEXP:NiagaraFloat=&self:PlatFromSelect(
     Default:NiagaraFloat=50
     IsTarget:NiagaraFloat=1.1
     TargetPlatform:NiagaraDataInterfacePlatformSet=DataInterface() {propertyValues {Platforms {
        qualityLevelMask=-32
        deviceProfileStates [
         {profileName=IOS, qualityLevelMask=4294967295, setQualityLevelMask=4294967295}
         {profileName=Android, qualityLevelMask=4294967295, setQualityLevelMask=4294967295}
         {profileName="Android_Mali_G72 ", qualityLevelMask=4294967295, setQualityLevelMask=4294967295}
         {profileName=Android_Galaxy_Z_Fold2_Adreno, qualityLevelMask=4294967295, setQualityLevelMask=4294967295}
         {profileName=Android_Galaxy_Z_Fold2_Adreno_LRU, qualityLevelMask=4294967295, setQualityLevelMask=4294967295}
         {profileName=Android_Galaxy_Z_Fold2_Adreno_Vulkan, qualityLevelMask=4294967295, setQualityLevelMask=4294967295}
         {profileName=Android_High, qualityLevelMask=4294967295, setQualityLevelMask=4294967295}
         {profileName=IOS_Epic, qualityLevelMask=4294967295, setQualityLevelMask=4294967295}
        ]
       }}}
    )
   ) {isSetParameters=true}
  ]
  particleSpawnScript [
   &FN_SetSpawnGroup(Number of Spawn Groups=2)
   &InitializeParticle(
    Lifetime=&Lerp_Float(
     A=&RandomRangeFloat(Minimum=0.6, Maximum=0.8)
     B=&RandomRangeFloat(Minimum=0.65, Maximum=0.9)
     Alpha=$System.RDive
    )
    Mass Mode=::NewEnumerator3
    Mesh Scale Mode=::NewEnumerator1
    Mesh Uniform Scale Max=0.3404
    Mesh Uniform Scale Min=0.1551
    Sprite Rotation Mode=::NewEnumerator2
    Sprite Size Mode=::NewEnumerator1
    Uniform Sprite Size Max=18
    Uniform Sprite Size Min=9
    UsePositionOffset=false
    Write Material Random=true
   )
   &CylinderLocation(Cylinder Height=0.5, Cylinder Radius=5, MaskSpawnGroup=false, Override Seed=false, UseOwnerScale=true)
   &AddVelocity(Velocity=&Lerp_Vector(A=<0 0 160>, B=<0 0 300>, Alpha=$System.RDive))
   &InitialMeshOrientation(Use Rotation Vector=false)
  ]
  particleUpdateScript [
   &ParticleState()
   &CurlNoiseForce(
    Noise Frequency=500
    Noise Strength=&FloatFromCurve(FloatCurve=DataInterface() {propertyValues {Curve {keys [<at 0 0.5>,<at 0.5 0.5>,<at 1 2>]}}}, Scale Curve=500)
    Pan Noise=true
   )
   &SolveForcesAndVelocity(Clamp Velocity=false, Limit Acceleration=false)
   &ScaleMeshSize(Scale Factor=&VectorFromFloat(Value=&FloatFromCurve(FloatCurve=DataInterface() {propertyValues {Curve {keys [
          {interpMode=RCIM_Cubic, leaveTangent=19.4774, tangentMode=RCTM_Break, time=0, value=0}
          {arriveTangent=0.2908, interpMode=RCIM_Cubic, tangentMode=RCTM_Break, time=0.1447, value=1}
          {arriveTangent=-0.3858, interpMode=RCIM_Cubic, leaveTangent=-0.3858, tangentMode=RCTM_User, time=0.7, value=0.8}
          {arriveTangent=-4.4397, interpMode=RCIM_Cubic, leaveTangent=-4.4397, tangentMode=RCTM_User, time=1, value=0}
         ]}}})))
   &ScaleMeshSize(Scale Factor=&VectorFromFloat(Value=0.7)) {enabled=false, module=ScaleMeshSize002}
   &SkeletalMeshLocation(
    Bone / Socket Index=$Particles.SpawnGroup
    Bone Sampling Mode=::NewEnumerator4
    MaskSpawnGroup=false
    Position Sampling=::NewEnumerator2
    Skeletal Mesh=$User.SmokeStarSkeletalMesh
   )
   /SetVariables(Particles.NPosition:Vector3f=&Add_Vector(
     A=&Add_Vector(A=Linked() {linkedVariable {name=Particles.Position, type=NiagaraPosition}}, B=<0 0 0>)
     B=&Multiply_VectorByFloat(
      Vector=&Lerp_Vector(
       A=<0 0 0>
       B=<7 0 0>
       Alpha=&ClampFloat(Float=&Divide_Float(
         A=&MakeFloatFromVector(Vector=Linked() {linkedVariable {name=Particles.Position, type=NiagaraPosition}}, Channel=::NewEnumerator2)
         B=60
        ))
      )
      Float=&Sine(
       Normalized Angle=&Add_Float(
        A=&Add_Float(
         A=&Multiply_Float(A=$System.Age, B=1)
         B=&MakeCustomFloatFromBool(
          Bool=&SetBoolByIntComparison(A=&SpawnGroup(), B=0, Comparison Type=::NewEnumerator4)
          True Float=0.43
          False Float=0
         )
        )
        B=&Divide_Float(
         A=&MakeFloatFromVector(Vector=Linked() {linkedVariable {name=Particles.Position, type=NiagaraPosition}}, Channel=::NewEnumerator2)
         B=150
        )
       )
       Period=1
       Scale=1
       Bias=0
      )
     )
    )) {isSetParameters=true, module=SetVariables_#9}
   /SetVariables(Particles.NPosition:Vector3f=&Add_Vector(
     A=/Niagara/DynamicInputs/Math/TransformPositionByMatrix(
      Position:Vector3f=$Particles.NPosition
      TransformationMatrix:NiagaraMatrix=/Niagara/DynamicInputs/Math/MakeTransform(
       Inverse:NiagaraBool=false
       Position:Vector3f=?
       Local Offset:Vector3f=<0 0 0>
       Scale:Vector3f=<1 1 1>
       Rotation Interface:ENiagara_RotationDefinitionApproach=::NewEnumerator2
       Axis:Vector3f=?
       Quaternion:Quat4f=?
       Rotation Matrix:NiagaraMatrix=?
       X Vector:Vector3f=?
       Yaw:NiagaraFloat=0
       Angle:NiagaraFloat=?
       Pitch:NiagaraFloat=&Lerp_Float(
        A=&MakeCustomFloatFromBool(
         Bool=&SetBoolByIntComparison(A=$Particles.SpawnGroup, B=0, Comparison Type=::NewEnumerator4)
         True Float=5
         False Float=-5
        )
        B=&MakeCustomFloatFromBool(
         Bool=&SetBoolByIntComparison(A=$Particles.SpawnGroup, B=0, Comparison Type=::NewEnumerator4)
         True Float=5.5
         False Float=-5.5
        )
        Alpha=$System.Dive
       )
       Y Vector:Vector3f=?
       Roll:NiagaraFloat=&Lerp_Float(
        A=-10
        B=&MakeCustomFloatFromBool(Bool=$User.bIsFrontEndPreview, True Float=-45, False Float=0)
        Alpha=$System.Dive
       )
       Z Vector:Vector3f=?
      )
     )
     B=Linked() {linkedVariable {name=Particles.SkeletalMeshLocation.SampledPosition, type=NiagaraPosition}}
    )) {isSetParameters=true}
   &ScaleMeshSize(Scale Factor=&VectorFromFloat(Value=&Lerp_Float(A=0.73, B=0.67, Alpha=$System.RDive))) {module=ScaleMeshSize001}
   /SetVariables(Particles.NScale:Vector3f=&Multiply_VectorByFloat(Vector=$Particles.Scale, Float=1.2)) {isSetParameters=true, module=SetVariables_#10}
   &UpdateMeshOrientation(Rotation Rate=30, Rotation Vector=&RandomVector(Vector Scale=1))
  ]
  rendererClasses [&NiagaraMeshRendererProperties]
  renderers [
   {rendererClass=&NiagaraMeshRendererProperties, rendererIndex=0}
   {rendererClass=&NiagaraMeshRendererProperties, rendererIndex=1}
   {rendererClass=&NiagaraMeshRendererProperties, rendererIndex=2}
  ]
 }
 Fountain001 {
  emitterUpdateScript [
   &EmitterState(Life Cycle Mode=::NewEnumerator1, Loop Duration=2, UseLoopDelay=false)
   &SpawnRate(SpawnRate=&Lerp_Float(A=20, B=40, Alpha=$System.RDive), Use Spawn Probability=false)
   /SetVariables(
    Emitter.EdgeThickness=0.45
    Emitter.FresnelThickEXP:NiagaraFloat=&self:PlatFromSelect(
     Default:NiagaraFloat=50
     IsTarget:NiagaraFloat=1.1
     TargetPlatform:NiagaraDataInterfacePlatformSet=DataInterface() {propertyValues {Platforms {
        qualityLevelMask=-32
        deviceProfileStates [
         {profileName=IOS, qualityLevelMask=4294967295, setQualityLevelMask=4294967295}
         {profileName=Android, qualityLevelMask=4294967295, setQualityLevelMask=4294967295}
         {profileName="Android_Mali_G72 ", qualityLevelMask=4294967295, setQualityLevelMask=4294967295}
         {profileName=Android_Galaxy_Z_Fold2_Adreno, qualityLevelMask=4294967295, setQualityLevelMask=4294967295}
         {profileName=Android_Galaxy_Z_Fold2_Adreno_LRU, qualityLevelMask=4294967295, setQualityLevelMask=4294967295}
         {profileName=Android_Galaxy_Z_Fold2_Adreno_Vulkan, qualityLevelMask=4294967295, setQualityLevelMask=4294967295}
         {profileName=Android_High, qualityLevelMask=4294967295, setQualityLevelMask=4294967295}
         {profileName=IOS_Epic, qualityLevelMask=4294967295, setQualityLevelMask=4294967295}
        ]
       }}}
    )
   ) {isSetParameters=true}
  ]
  particleSpawnScript [
   &FN_SetSpawnGroup(Number of Spawn Groups=2)
   &InitializeParticle(
    Lifetime=&Lerp_Float(A=0.45, B=0.4, Alpha=$System.RDive)
    Mass Mode=::NewEnumerator3
    Mesh Scale Mode=::NewEnumerator1
    Mesh Uniform Scale Max=0.2223
    Mesh Uniform Scale Min=0.1112
    Sprite Rotation Mode=::NewEnumerator2
    Sprite Size Mode=::NewEnumerator1
    Uniform Sprite Size Max=18
    Uniform Sprite Size Min=9
    UsePositionOffset=false
    Write Material Random=true
   )
   &CylinderLocation(Cylinder Height=0.5, Cylinder Radius=3, MaskSpawnGroup=false, Override Seed=false, UseOwnerScale=true)
   &AddVelocity(Velocity=&Lerp_Vector(A=<0 0 80>, B=<0 0 200>, Alpha=$System.RDive))
   &InitialMeshOrientation(Use Rotation Vector=false)
  ]
  particleUpdateScript [
   &ParticleState()
   &CurlNoiseForce(Noise Frequency=500, Noise Strength=50, Pan Noise=true)
   &SolveForcesAndVelocity(Clamp Velocity=false, Limit Acceleration=false)
   &ScaleMeshSize(Scale Factor=&VectorFromFloat(Value=&FloatFromCurve(FloatCurve=DataInterface() {propertyValues {Curve {keys [
          {interpMode=RCIM_Cubic, leaveTangent=19.4774, tangentMode=RCTM_Break, time=0, value=0}
          {interpMode=RCIM_Cubic, tangentMode=RCTM_Break, time=0.1447, value=1}
          {arriveTangent=-0.5305, interpMode=RCIM_Cubic, leaveTangent=-0.5305, tangentMode=RCTM_User, time=0.7, value=0.8}
          {arriveTangent=-4.7219, interpMode=RCIM_Cubic, leaveTangent=-4.7219, tangentMode=RCTM_User, time=1, value=0}
         ]}}})))
   &ScaleMeshSize(Scale Factor=&VectorFromFloat(Value=0.7)) {enabled=false, module=ScaleMeshSize002}
   &SkeletalMeshLocation(
    Bone / Socket Index=&Add_Integer(A=$Particles.SpawnGroup, B=2)
    Bone Sampling Mode=::NewEnumerator4
    MaskSpawnGroup=false
    Position Sampling=::NewEnumerator2
    Skeletal Mesh=$User.SmokeStarSkeletalMesh
   )
   /SetVariables(Particles.NPosition:Vector3f=&TransformPositionByMatrix(
     Position:Vector3f=Linked() {linkedVariable {name=Particles.Position, type=NiagaraPosition}}
     TransformationMatrix:NiagaraMatrix=&MakeTransform(
      Inverse:NiagaraBool=false
      Position:Vector3f=?
      Local Offset:Vector3f=<0 0 0>
      Scale:Vector3f=<1 1 1>
      Rotation Interface:ENiagara_RotationDefinitionApproach=::NewEnumerator2
      Axis:Vector3f=?
      Quaternion:Quat4f=?
      Rotation Matrix:NiagaraMatrix=?
      X Vector:Vector3f=?
      Yaw:NiagaraFloat=0
      Angle:NiagaraFloat=?
      Pitch:NiagaraFloat=0
      Y Vector:Vector3f=?
      Roll:NiagaraFloat=&MakeCustomFloatFromBool(Bool=$User.bIsFrontEndPreview, True Float=&Lerp_Float(A=0, B=-45, Alpha=$System.Dive), False Float=0)
      Z Vector:Vector3f=?
     )
    )) {isSetParameters=true}
   /SetVariables(Particles.NPosition:Vector3f=&Add_Vector(A=$Particles.NPosition, B=Linked() {linkedVariable {name=Particles.SkeletalMeshLocation.SampledPosition, type=NiagaraPosition}})) {isSetParameters=true, module=SetVariables_#9}
   &ScaleMeshSize(Scale Factor=&VectorFromFloat(Value=&Lerp_Float(A=0.7, B=0.55, Alpha=$System.RDive))) {module=ScaleMeshSize001}
   /SetVariables(Particles.NScale:Vector3f=&Multiply_VectorByFloat(Vector=$Particles.Scale, Float=1.3)) {isSetParameters=true, module=SetVariables_#10}
   &UpdateMeshOrientation(Rotation Rate=30, Rotation Vector=&RandomVector(Vector Scale=1))
  ]
  rendererClasses [&NiagaraMeshRendererProperties]
  renderers [
   {rendererClass=&NiagaraMeshRendererProperties, rendererIndex=0}
   {rendererClass=&NiagaraMeshRendererProperties, rendererIndex=1}
   {rendererClass=&NiagaraMeshRendererProperties, rendererIndex=2}
  ]
 }
}
summary {
 systemName=NS_Contrail_TeriyakiToon
 emitters [
  {emitterName=Fountain, rendererClasses [&NiagaraMeshRendererProperties]}
  {emitterName=Fountain001, rendererClasses [&NiagaraMeshRendererProperties]}
  {emitterName=Empty, rendererClasses [&NiagaraSpriteRendererProperties]}
  {emitterName=Empty001, rendererClasses [&NiagaraMeshRendererProperties]}
  {emitterName=Empty002, rendererClasses [&NiagaraSpriteRendererProperties]}
 ]
 userVariables [
  User.bIsFrontEnd {type=NiagaraBool}
  User.bIsFrontEndPreview {type=NiagaraBool}
  User.SkeletalMesh {
   type=NiagaraDataInterfaceSkeletalMesh
   defaultValue {
    struct=NiagaraExt_VariableValue_DataInterface
    value {dataInterface=&self:NiagaraDataInterfaceSkeletalMesh_1}
   }
  }
  User.SmokeStarSkeletalMesh {
   type=NiagaraDataInterfaceSkeletalMesh
   defaultValue {
    struct=NiagaraExt_VariableValue_DataInterface
    value {dataInterface=&self:NiagaraDataInterfaceSkeletalMesh_0}
   }
  }
 ]
}
